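# Sound cue for the thrust-stop sound of the MarkerDeer Tin glider.
# A Random node picks one of three equally weighted thrust-stop wave variants, which feeds a Modulator that adds slight pitch variation.
# It uses the glider sound class, player glider thrust attenuation and reduced volume.
# Concurrency references the glider thrust set and also enables an override that caps instances and stops the oldest.
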SoundCue /BRCosmetics/Sounds/Gliders/MarkerDeer_Tin/GliderThrustStop_MarkerDeer_Tin_Cue
root=Modulator_0
nodes [
 {id=Modulator_0, children [Random_0], properties {pitchMin=0.99}}
 {id=Random_0, children [WavePlayer_2,WavePlayer_3,WavePlayer_4], properties {weights [1,1,1]}}
 {id=WavePlayer_2, properties {soundWaveAssetPtr=&/IP_Glider_MarkerDeer_Tin_ThrustStop_01}}
 {id=WavePlayer_3, properties {soundWaveAssetPtr=&/IP_Glider_MarkerDeer_Tin_ThrustStop_02}}
 {id=WavePlayer_4, properties {soundWaveAssetPtr=&/IP_Glider_MarkerDeer_Tin_ThrustStop_03}}
]
properties {
 attenuationSettings=&Fort_Player_Glider_Thrust_Athena_Att
 bOverrideConcurrency=true
 maxDistance=12500
 soundClassObject=/Game/Sounds/FortSoundClassesAndModes/Fort_Gliders
 volumeMultiplier=0.45
 concurrencyOverrides {maxCount=3, resolutionRule=StopOldest}
 concurrencySet [&Athena_Glider_Thrust_Conc]
}
