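# Sound cue for the wings-close sound of the MarkerDeer Tin cosmetic glider.
# A Random node picks one of three equally weighted wave variants and feeds a Modulator that adds slight variation.
# It plays at reduced volume through the glider sound class, close-glider property sheet and player glider thrust attenuation.
# Its concurrency override caps simultaneous instances and stops the oldest, taking precedence over the assigned thrust concurrency set.
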SoundCue /BRCosmetics/Sounds/Gliders/MarkerDeer_Tin/IP_Glider_MarkerDeer_Tin_Wings_Close_Cue
root=Modulator_0
nodes [
 {id=Modulator_0, children [Random_0], properties {pitchMin=0.99}}
 {id=Random_0, children [WavePlayer_0,WavePlayer_1,WavePlayer_2], properties {weights [1,1,1]}}
 {id=WavePlayer_0, properties {soundWaveAssetPtr=&/IP_Glider_MarkerDeer_Tin_Wings_Close_01}}
 {id=WavePlayer_1, properties {soundWaveAssetPtr=&/IP_Glider_MarkerDeer_Tin_Wings_Close_03}}
 {id=WavePlayer_2, properties {soundWaveAssetPtr=&/IP_Glider_MarkerDeer_Tin_Wings_Close_02}}
]
properties {
 attenuationSettings=&Fort_Player_Glider_Thrust_Athena_Att
 audioPropertiesSheet=/BRCosmetics/Sounds/Settings/PropertySheets/Glider/APS_Cosmetic_Glider_Close
 bOverrideConcurrency=true
 maxDistance=12500
 soundClassObject=/Game/Sounds/FortSoundClassesAndModes/Fort_Gliders
 volumeMultiplier=0.3
 concurrencyOverrides {maxCount=3, resolutionRule=StopOldest}
 concurrencySet [&Athena_Glider_Thrust_Conc]
}
